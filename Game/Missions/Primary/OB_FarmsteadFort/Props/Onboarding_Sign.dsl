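; An onboarding sign prop for the Farmstead Fort mission that lights up when objective 8 activates.
; It sets Glow off at BeginPlay and on when it receives the OBJ_8_Activate mission event.
; The Glow RepNotify creates a dynamic material instance in slot 0 of the static mesh from the lit or unlit open-sign material, so the visual follows the replicated state.
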(event EventBeginPlay
  (Variables|Default|SetGlow false))

(event MissionHelpers|HandleMissionEvent%28v2%29:OBJ_8_Activate
  (params (InMissionGuid :type Guid%20Structure) (ObjectiveHandle :type Gameplay%20Tag%20Container%20Structure) (Params :type Fort%20Mission%20Event%20Params%20Object%20Reference) (DO_NOT_USE_THIS_OR_VARIABLES_BELOW :type Boolean) (EventFocus :type Object%20Reference) (EventContent :type Data%20Asset%20Object%20Reference) (EventInstigator :type Actor%20Object%20Reference) (GenericInt :type Integer) (GenericFloat :type Float%20%28single-precision%29) (GenericText :type Text%20%28by%20ref%29) (GameplayTags :type Gameplay%20Tag%20Container%20Structure) (MissionEvent :type Fort%20Mission%20Event%20Structure%20%28by%20ref%29))
  (Variables|Default|SetGlow true))

(fn OnRep_Glow
  (bind _staticmeshcomponent (Variables|Debug|GetStaticMeshComponent))
  (if (Variables|Default|GetGlow)
    (Rendering|Material|CreateDynamicMaterialInstance _staticmeshcomponent 0 "/Game/Environments/Props/Signs/Materials/MI_OpenSign.MI_OpenSign" "None" :!class /Script/Engine.PrimitiveComponent)
    (else
      (Rendering|Material|CreateDynamicMaterialInstance _staticmeshcomponent 0 "/Game/Environments/Props/Signs/Materials/MI_OpenSignOff.MI_OpenSignOff" "None" :!class /Script/Engine.PrimitiveComponent))))

(fn ConstructionScript)
